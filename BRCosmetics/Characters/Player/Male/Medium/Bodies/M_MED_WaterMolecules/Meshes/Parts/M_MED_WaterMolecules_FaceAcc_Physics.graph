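# Physics asset for the face accessory mesh of the male medium WaterMolecules character body, defining capsule physics for its dangling hair bones.
# Two-bone left and right side chains and a three-bone front chain each hang from a small MSTR root capsule.
# Chain joints have zero twist limit, with tighter swing limits on the front chain than the sides.
# Roots link to spine_05, which carries two large angled capsules, through constraints with free swing and twist.
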PhysicsAsset /BRCosmetics/Characters/Player/Male/Medium/Bodies/M_MED_WaterMolecules/Meshes/Parts/M_MED_WaterMolecules_FaceAcc_Physics
bodies {
 MSTR_hairSide_l [{shapeType=Capsule, length=0.505, radius=0.505, shapeName=MSTR_hairSide_l_capsule, rotation=<rot 0 -90 0>}]
 MSTR_hairSide_r [{shapeType=Capsule, length=0.505, radius=0.505, shapeName=MSTR_hairSide_r_capsule, rotation=<rot 0 -90 0>}]
 MSTR_hair_fr_04 [{shapeType=Capsule, length=0.505, radius=0.505, shapeName=MSTR_hair_fr_04_capsule, rotation=<rot 0 -90 0>}]
 dyn_hairSide_l_01 [{shapeType=Capsule, length=3, radius=1, shapeName=dyn_hairSide_l_01_capsule, center=<3 0 0>, rotation=<rot -90 0 0>}]
 dyn_hairSide_l_02 [{shapeType=Capsule, length=3, radius=1, shapeName=dyn_hairSide_l_02_capsule, center=<3 0 0>, rotation=<rot -90 0 0>}]
 dyn_hairSide_r_01 [{shapeType=Capsule, length=3, radius=1, shapeName=dyn_hairSide_r_01_capsule, center=<3 0 0>, rotation=<rot -90 0 0>}]
 dyn_hairSide_r_02 [{shapeType=Capsule, length=3, radius=1, shapeName=dyn_hairSide_r_02_capsule, center=<3 0 0>, rotation=<rot -90 0 0>}]
 dyn_hair_fr_01 [{shapeType=Capsule, length=3, radius=1, shapeName=dyn_hair_fr_01_capsule, center=<3 0 0>, rotation=<rot -90 0 0>}]
 dyn_hair_fr_02 [{shapeType=Capsule, length=3, radius=1, shapeName=dyn_hair_fr_02_capsule, center=<3 0 0>, rotation=<rot -90 0 0>}]
 dyn_hair_fr_03 [{shapeType=Capsule, length=3, radius=1, shapeName=dyn_hair_fr_03_capsule, center=<3 0 0>, rotation=<rot -90 0 0>}]
 spine_05 [
  {
   shapeType=Capsule
   length=16.3012
   radius=8
   shapeName=spine_05_capsule
   center=<14.7979 13.2219 -15.23>
   rotation=<rot 0.2946 -0.1187 68.2831>
  }
  {shapeType=Capsule, length=16, radius=8, shapeName=spine_05_capsule, center=<17.129 -3.7075 -13.7286>, rotation=<rot 0 -0.0001 -64.371>}
 ]
}
constraints [
 {bone1Name=dyn_hairSide_l_01, bone2Name=MSTR_hairSide_l, swing1LimitDegrees=30, swing2LimitDegrees=30, twistLimitDegrees=0}
 {bone1Name=dyn_hairSide_l_02, bone2Name=dyn_hairSide_l_01, swing1LimitDegrees=30, swing2LimitDegrees=30, twistLimitDegrees=0}
 {bone1Name=dyn_hairSide_r_01, bone2Name=MSTR_hairSide_r, swing1LimitDegrees=30, swing2LimitDegrees=30, twistLimitDegrees=0}
 {bone1Name=dyn_hairSide_r_02, bone2Name=dyn_hairSide_r_01, swing1LimitDegrees=30, swing2LimitDegrees=30, twistLimitDegrees=0}
 {bone1Name=dyn_hair_fr_01, bone2Name=MSTR_hair_fr_04, swing1LimitDegrees=5, swing2LimitDegrees=5, twistLimitDegrees=0}
 {bone1Name=dyn_hair_fr_02, bone2Name=dyn_hair_fr_01, swing1LimitDegrees=5, swing2LimitDegrees=5, twistLimitDegrees=0}
 {bone1Name=dyn_hair_fr_03, bone2Name=dyn_hair_fr_02, swing1LimitDegrees=5, swing2LimitDegrees=5, twistLimitDegrees=0}
 {
  bone1Name=MSTR_hairSide_l
  bone2Name=spine_05
  swing1LimitDegrees=45
  swing1Motion=Free
  swing2LimitDegrees=45
  swing2Motion=Free
  twistLimitDegrees=45
  twistMotion=Free
 }
 {
  bone1Name=MSTR_hairSide_r
  bone2Name=spine_05
  swing1LimitDegrees=45
  swing1Motion=Free
  swing2LimitDegrees=45
  swing2Motion=Free
  twistLimitDegrees=45
  twistMotion=Free
 }
 {
  bone1Name=MSTR_hair_fr_04
  bone2Name=spine_05
  swing1LimitDegrees=45
  swing1Motion=Free
  swing2LimitDegrees=45
  swing2Motion=Free
  twistLimitDegrees=45
  twistMotion=Free
 }
]
